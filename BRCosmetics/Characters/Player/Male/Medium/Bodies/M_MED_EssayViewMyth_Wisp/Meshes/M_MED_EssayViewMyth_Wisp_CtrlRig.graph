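# Control Rig for the Wisp male medium body that adds procedural corrections to its shorts, lanyards, pendant and right wristband.
# After the shared DeformRig function, weighted parent constraints place accessory bones, and remapped knee bend adjusts the shorts.
# Wristband offsets blend by deform-rig anim attributes and by distances between wrist reference controls created at setup; spherical pose readers sit unconnected.
# When IsFrontEnd is true, the pendant aims at a constrained, offset helper bone.
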ControlRigBlueprint /BRCosmetics/Characters/Player/Male/Medium/Bodies/M_MED_EssayViewMyth_Wisp/Meshes/M_MED_EssayViewMyth_Wisp_CtrlRig
graphs [
 New Function_ContainedGraph {nodes [
   RigVMFunctionEntryNode()
   > &RigUnit_ModifyTransforms(
    ItemToModify=((Item=(Type=Bone,Name="base_shorts_fmd_l_02"),Transform=(Rotation=(X=-0.2164,Y=0,Z=0,W=0.9763),Translation=(X=-20,Y=0,Z=0),Scale3D=(X=0.5,Y=1,Z=1))))
    WeightMinimum=0
    Weight=RigVMVariableNode() {variable=CalfRot_L}.Value
   )
   > &RigUnit_ModifyTransforms(
    ItemToModify=((Item=(Type=Bone,Name="base_shorts_fmd_r_02"),Transform=(Rotation=(X=0.2164,Y=0,Z=0,W=0.9763),Translation=(X=-20,Y=0,Z=0),Scale3D=(X=0.5,Y=1,Z=1))))
    WeightMinimum=0
    Weight=RigVMVariableNode() {variable=CalfRot_R}.Value
   )
   > RigVMFunctionReturnNode()
  ]}
 New Function_ContainedGraph {nodes [
   RigVMFunctionEntryNode()
   > deltaRotZ = RigVMFunctionReferenceNode(Child=calf_l, Parent=thigh_l, Source Max=120, Source Min=80)
   > RigVMVariableNode(Value=@deltaRotZ.Delta) {variable=CalfRot_L}
   > deltaRotZ_1 = RigVMFunctionReferenceNode(Child=calf_r, Parent=thigh_r, Source Max=120, Source Min=80)
   > RigVMVariableNode(Value=@deltaRotZ_1.Delta) {variable=CalfRot_R}
   > RigVMFunctionReturnNode()
  ]}
 New Function_ContainedGraph {nodes [
   n1 = RigVMFunctionEntryNode()
   RigVMFunction_MathQuaternionToEuler = &RigVMFunction_MathQuaternionToEuler(
    RotationOrder=ZYX
    Value=&RigUnit_GetRelativeTransformForItem(
     Child=(Type=Bone,Name="None")
     Parent=(Type=Bone,Name="None")
     bChildInitial=false
     bParentInitial=False
     Child.Name=@n1.Child
     Parent.Name=@n1.Parent
    ).RelativeTransform.Rotation
   )
   Remap = &RigVMFunction_MathDoubleRemap(
    TargetMaximum=1
    TargetMinimum=0
    Value=@RigVMFunction_MathQuaternionToEuler.Result.Z
    SourceMinimum=@n1.Source Min
    SourceMaximum=@n1.Source Max
   )
   > RigVMFunctionReturnNode(Delta=@Remap.Result)
  ]}
 New Function_ContainedGraph {nodes [
   RigVMFunctionEntryNode()
   > &RigUnit_ParentConstraint(
    AdvancedSettings=(RotationOrderForFilter=XZY)
    Child=(Type=Bone,Name="dyn_lanyard_l")
    Parents="((Item=(Type=Bone,Name=\"deform_groin_r\"),Weight=0.0057),(Item=(Type=Bone,Name=\"thigh_twist_01_r\"),Weight=0.0106),(Item=(Type=Bone,Name=\"thigh_twist_01_l\"),Weight=0.0287),(Item=(Type=Bone,Name=\"spine_03\"),Weight=0.0297),(Item=(Type=Bone,Name=\"deform_groin_l ...(truncated)"
    Weight=1
   )
   > &RigUnit_ParentConstraint(
    AdvancedSettings=(RotationOrderForFilter=XZY)
    Child=(Type=Bone,Name="dyn_lanyard_r")
    Parents="((Item=(Type=Bone,Name=\"deform_groin_r\"),Weight=0.016),(Item=(Type=Bone,Name=\"deform_groin_l\"),Weight=0.0171),(Item=(Type=Bone,Name=\"thigh_twist_01_r\"),Weight=0.0186),(Item=(Type=Bone,Name=\"thigh_twist_01_l\"),Weight=0.0192),(Item=(Type=Bone,Name=\"spine_03\" ...(truncated)"
    Weight=1
   )
   > &RigUnit_ParentConstraint(
    AdvancedSettings=(RotationOrderForFilter=XZY)
    Child=(Type=Bone,Name="base_pendant")
    Parents=((Item=(Type=Bone,Name="neck_01"),Weight=0.0028),(Item=(Type=Bone,Name="spine_02"),Weight=0.0086),(Item=(Type=Bone,Name="spine_03"),Weight=0.0326),(Item=(Type=Bone,Name="spine_04"),Weight=0.1247),(Item=(Type=Bone,Name="spine_05"),Weight=0.8313))
    Weight=1
   )
   > &RigUnit_ParentConstraint(
    AdvancedSettings=(RotationOrderForFilter=XZY)
    Child=(Type=Bone,Name="base_wristband_r")
    Parents="((Item=(Type=Bone,Name=\"thumb_01_r\"),Weight=0.005),(Item=(Type=Bone,Name=\"ring_metacarpal_r\"),Weight=0.0143),(Item=(Type=Bone,Name=\"index_metacarpal_r\"),Weight=0.0187),(Item=(Type=Bone,Name=\"pinky_metacarpal_r\"),Weight=0.0228),(Item=(Type=Bone,Name=\"lowera ...(truncated)"
    Weight=1
   )
   > RigVMFunctionReturnNode()
  ]}
 New Function_ContainedGraph {nodes [
   RigVMFunctionEntryNode()
   > &RigUnit_ParentConstraint(
    AdvancedSettings=(InterpolationType=Average,RotationOrderForFilter=XZY)
    Child=(Type=Bone,Name="dyn_wristband_r")
    Parents=((Item=(Type=Bone,Name="lowerarm_r"),Weight=0.5),(Item=(Type=Bone,Name="hand_r"),Weight=0.5))
    Weight=1
   )
   > &RigUnit_ModifyTransforms(
    ItemToModify=((Item=(Type=Bone,Name="base_wristband_r"),Transform=(Rotation=(X=0,Y=0,Z=0,W=1),Translation=(X=0.7,Y=-0.3,Z=0),Scale3D=(X=1,Y=1,Z=1))))
    Weight=1
    WeightMinimum=0
   )
   > &RigUnit_ModifyTransforms(
    ItemToModify=((Item=(Type=Bone,Name="base_wristband_r"),Transform=(Rotation=(X=0,Y=0,Z=-0.0872,W=0.9962),Translation=(X=-1,Y=-1,Z=0),Scale3D=(X=0.9,Y=1.1,Z=1))))
    WeightMinimum=0
    Weight=RigVMDispatchNode(BoneName=root, Default=0, Name=deformRig_r_wrist_up_alpha) {template="DISPATCH_RigDispatch_GetAnimAttribute(in Name,in BoneName,in Default,out Value,out Found)"}.Value
   )
   > &RigUnit_ModifyTransforms(
    ItemToModify=((Item=(Type=Bone,Name="base_wristband_r"),Transform=(Rotation=(X=0,Y=0,Z=-0.0872,W=0.9962),Translation=(X=-1,Y=1,Z=0),Scale3D=(X=0.9,Y=1,Z=1))))
    WeightMinimum=0
    Weight=RigVMDispatchNode(BoneName=root, Default=0, Name=deformRig_r_wrist_dn_alpha) {template="DISPATCH_RigDispatch_GetAnimAttribute(in Name,in BoneName,in Default,out Value,out Found)"}.Value
   )
   n1 = &RigUnit_GetTransform(Item=(Type=Control,Name="wrist_ref_ctrl"), bInitial=False)
   RigVMFunction_MathVectorDistance = &RigVMFunction_MathVectorDistance(
    B=@n1.Transform.Translation
    A=&RigUnit_GetTransform(Item=(Type=Control,Name="wrist_down_ctrl"), bInitial=False).Transform.Translation
   )
   Remap = &RigVMFunction_MathDoubleRemap(SourceMaximum=4, SourceMinimum=15, TargetMaximum=1, TargetMinimum=0, Value=@RigVMFunction_MathVectorDistance.Result)
   > &RigUnit_ModifyTransforms(
    ItemToModify=((Item=(Type=Bone,Name="base_wristband_r"),Transform=(Rotation=(X=0,Y=0.0436,Z=0,W=0.999),Translation=(X=-0.8,Y=0,Z=-0.5),Scale3D=(X=0.9,Y=1,Z=1))))
    WeightMinimum=0
    Weight=@Remap.Result
   )
   n2 = &RigUnit_GetTransform(Item=(Type=Control,Name="wrist_up_ctrl"), bInitial=False)
   > &RigUnit_ModifyTransforms(
    ItemToModify=((Item=(Type=Bone,Name="base_wristband_r"),Transform=(Rotation=(X=0,Y=-0.0436,Z=0,W=0.999),Translation=(X=0,Y=0,Z=0.5),Scale3D=(X=1,Y=1,Z=1))))
    WeightMinimum=0
    Weight=&RigVMFunction_MathDoubleRemap(
     SourceMaximum=3
     SourceMinimum=14
     TargetMaximum=1
     TargetMinimum=0
     Value=&RigVMFunction_MathVectorDistance(A=@n1.Transform.Translation, B=@n2.Transform.Translation).Result
    ).Result
   )
   > RigVMFunctionReturnNode()
   n3 = &RigUnit_SphericalPoseReader(
    ActiveRegionScaleFactors=(PositiveWidth=0.5,NegativeWidth=1,PositiveHeight=0.5,NegativeHeight=0.5)
    ActiveRegionSize=0.1
    Debug=(bDrawDebug=true,bDraw2D=false,bDrawLocalAxes=true,DebugScale=25,DebugSegments=20,DebugThickness=0.25)
    DriverAxis=(X=1,Y=0,Z=0)
    DriverItem=(Type=Bone,Name="hand_r")
    ExecutePin=()
    FalloffRegionScaleFactors=(PositiveWidth=1,NegativeWidth=0,PositiveHeight=0.5,NegativeHeight=0.5)
    FalloffSize=0.2
    FlipHeightScaling=False
    FlipWidthScaling=true
    OptionalParentItem=(Type=Bone,Name="None")
    RotationOffset=(X=30,Y=-130,Z=0)
   )
   > n4 = &RigUnit_SphericalPoseReader(
    ActiveRegionScaleFactors=(PositiveWidth=0.5,NegativeWidth=1,PositiveHeight=0.5,NegativeHeight=0.5)
    ActiveRegionSize=0.1
    Debug=(bDrawDebug=true,bDraw2D=false,bDrawLocalAxes=true,DebugScale=25,DebugSegments=20,DebugThickness=0.25)
    DriverAxis=(X=1,Y=0,Z=0)
    DriverItem=(Type=Bone,Name="hand_r")
    FalloffRegionScaleFactors=(PositiveWidth=1,NegativeWidth=0,PositiveHeight=0.5,NegativeHeight=0.5)
    FalloffSize=0.2
    FlipHeightScaling=False
    FlipWidthScaling=false
    OptionalParentItem=(Type=Bone,Name="None")
    RotationOffset=(X=-20,Y=-30,Z=0)
   )
   > &RigUnit_ModifyTransforms(
    ItemToModify=((Item=(Type=Bone,Name="base_wristband_r"),Transform=(Rotation=(X=0,Y=0.0436,Z=0,W=0.999),Translation=(X=0.1,Y=0,Z=-0.5),Scale3D=(X=0.9,Y=1,Z=1))))
    WeightMinimum=0
    Weight=@n4.OutputParam
   )
   > &RigUnit_ModifyTransforms(
    ItemToModify=((Item=(Type=Bone,Name="base_wristband_r"),Transform=(Rotation=(X=0,Y=-0.0436,Z=0,W=0.999),Translation=(X=0,Y=0,Z=0.5),Scale3D=(X=1,Y=1,Z=1))))
    WeightMinimum=0
    Weight=@n3.OutputParam
   )
  ]}
 RigVMFunctionLibrary {nodes [
   ShortsSetup = RigVMCollapseNode(ExecuteContext=())
   Variables = RigVMCollapseNode(ExecuteContext=())
   constraints = RigVMCollapseNode(ExecuteContext=())
   deltaRotZ = RigVMCollapseNode(ExecuteContext=())
   watchSetup = RigVMCollapseNode(ExecuteContext=())
  ]}
 RigVMModel {nodes [
   n1 = &RigUnit_PrepareForExecution()
   &RigUnit_HierarchyImportFromSkeleton(NameSpace=None, bIncludeCurves=false, bIncludeMeshSockets=false, bIncludeVirtualBones=True, ExecutePin=@n1.ExecutePin)
   > Control Stack at Item = RigVMFunctionReferenceNode(
    Add Bottom Null=False
    Add Null=true
    Add Secondary Control=False
    Control Suffix=_ctrl
    Item=(Name="hand_r",Type=Bone)
    Name=wrist_ref
    Null Suffix=_null
    Parent=(Name="hand_r",Type=Bone)
    Secondary Settings="(Shape=(bVisible=True,Name=\"Default\",Color=(R=1,G=0,B=0,A=1),Transform=(Rotation=(X=0,Y=0,Z=0,W=1),Translation=(X=0,Y=0,Z=0),Scale3D=(X=1,Y=1,Z=1))),Proxy=(bIsProxy=False,ShapeVisibility=BasedOnSelection),DisplayName=\"None\",InitialSpace=LocalSpace,bUsePref ...(truncated)"
    Settings="(Shape=(bVisible=True,Name=\"Box_Solid\",Color=(R=1,G=0,B=0,A=1),Transform=(Rotation=(X=0,Y=0,Z=0,W=1),Translation=(X=0,Y=0,Z=0),Scale3D=(X=0.25,Y=0.25,Z=0.25))),Proxy=(bIsProxy=False,ShapeVisibility=BasedOnSelection),DisplayName=\"None\",InitialSpace=LocalSpa ...(truncated)"
   )
   > SetControlOffset = /Script/ControlRig.RigUnit_SetControlOffset(
    Offset=(Rotation=(X=0,Y=0,Z=0,W=1),Translation=(X=26,Y=0,Z=0),Scale3D=(X=1,Y=1,Z=1))
    Space=LocalSpace
    Control=@"Control Stack at Item".Control.Name
   ) {is=struct RigVMUnitNode, template="SetControlOffset::Execute(in Control,io Offset,in Space)"}
   > Control Stack at Item_1 = RigVMFunctionReferenceNode(
    Add Bottom Null=False
    Add Null=true
    Add Secondary Control=False
    Control Suffix=_ctrl
    Item=(Name="hand_r",Type=Bone)
    Name=wrist_up
    Null Suffix=_null
    Parent=(Name="lowerarm_r",Type=Bone)
    Secondary Settings="(Shape=(bVisible=True,Name=\"Default\",Color=(R=1,G=0,B=0,A=1),Transform=(Rotation=(X=0,Y=0,Z=0,W=1),Translation=(X=0,Y=0,Z=0),Scale3D=(X=1,Y=1,Z=1))),Proxy=(bIsProxy=False,ShapeVisibility=BasedOnSelection),DisplayName=\"None\",InitialSpace=LocalSpace,bUsePref ...(truncated)"
    Settings="(Shape=(bVisible=True,Name=\"Pyramid_Solid\",Color=(R=0,G=0,B=1,A=1),Transform=(Rotation=(X=0,Y=0,Z=0,W=1),Translation=(X=0,Y=0,Z=0),Scale3D=(X=0.5,Y=0.5,Z=0.5))),Proxy=(bIsProxy=False,ShapeVisibility=BasedOnSelection),DisplayName=\"None\",InitialSpace=LocalSp ...(truncated)"
   )
   > &RigUnit_SetControlOffset(
    Offset=(Rotation=(X=0,Y=0,Z=0,W=1),Translation=(X=17,Y=11,Z=-14),Scale3D=(X=1,Y=1,Z=1))
    Space=LocalSpace
    Control=@"Control Stack at Item_1".Control.Name
   ) {template="SetControlOffset::Execute(in Control,io Offset,in Space)"}
   > Control Stack at Item_2 = RigVMFunctionReferenceNode(
    Add Bottom Null=False
    Add Null=true
    Add Secondary Control=False
    Control Suffix=_ctrl
    Item=(Name="hand_r",Type=Bone)
    Name=wrist_down
    Null Suffix=_null
    Parent=(Name="lowerarm_r",Type=Bone)
    Secondary Settings="(Shape=(bVisible=True,Name=\"Default\",Color=(R=1,G=0,B=0,A=1),Transform=(Rotation=(X=0,Y=0,Z=0,W=1),Translation=(X=0,Y=0,Z=0),Scale3D=(X=1,Y=1,Z=1))),Proxy=(bIsProxy=False,ShapeVisibility=BasedOnSelection),DisplayName=\"None\",InitialSpace=LocalSpace,bUsePref ...(truncated)"
    Settings="(Shape=(bVisible=True,Name=\"Pyramid_Solid\",Color=(R=0,G=1,B=0,A=1),Transform=(Rotation=(X=0,Y=0,Z=0,W=1),Translation=(X=0,Y=0,Z=0),Scale3D=(X=0.5,Y=0.5,Z=0.5))),Proxy=(bIsProxy=False,ShapeVisibility=BasedOnSelection),DisplayName=\"None\",InitialSpace=LocalSp ...(truncated)"
   )
   > &RigUnit_SetControlOffset(
    Offset=(Rotation=(X=0,Y=0,Z=0,W=1),Translation=(X=10,Y=-8,Z=20),Scale3D=(X=1,Y=1,Z=1))
    Space=LocalSpace
    Control=@"Control Stack at Item_2".Control.Name
   ) {template="SetControlOffset::Execute(in Control,io Offset,in Space)"}
   RigUnit_BeginExecution = &RigUnit_BeginExecution()
   FN_DeformRig_FNC = RigVMFunctionReferenceNode(
    DeformRig_ Curve="(deformRig_l_upperarm_up_alpha_72_#1=true,deformRig_r_upperarm_up_alpha_73_#2=true,deformRig_l_upperarm_fr_alpha_74_#3=true,deformRig_r_upperarm_fr_alpha_75_CB51B80C4 ...(truncated)"
    DeformRig_Bool="(upperarm_1_#4=True,elbow_3_#5=True,wrist_7_#6=True,groin_8_#7=True,glute_17_#8=True,knee_9_40C408324CA261 ...(truncated)"
    DeformRig_Debug="(upperarm_debug=False,wrist_debug=False,lat_debug=False,groinGlute_debug=False,upperarm_debug_18_#4=False,wrist_debug_20_#6=False,groinGlute_debug_27_#7=False,lat_deb ...(truncated)"
    DeformRig_Value="(upperarm_up_18_#9=(Translation=(X=0,Y=0,Z=8),Scale3D=(X=0.7,Y=1,Z=1)),upperarm_fwd_fr_19_#10=(Translation=(X=0,Y=10,Z=0),Scale3D=(X=0.5,Y=1,Z=1)),upperarm_fwd_bk_20_9973AAD84EE0A1E349BC9E85067D4A3 ...(truncated)"
    ExecuteContext=@RigUnit_BeginExecution.ExecutePin
   )
   > constraints = RigVMFunctionReferenceNode()
   > Variables = RigVMFunctionReferenceNode()
   > ShortsSetup = RigVMFunctionReferenceNode()
   > watchSetup = RigVMFunctionReferenceNode()
   RigVMFunction_MathBoolToFloat = &RigVMFunction_MathBoolToFloat(Value=RigVMVariableNode() {variable=IsFrontEnd}.Value)
   > &RigUnit_ParentConstraint(
    AdvancedSettings=(RotationOrderForFilter=XZY)
    Child=(Type=Bone,Name="base_pendant_for_aim")
    Parents="((Item=(Type=Bone,Name=\"pelvis\"),Weight=0.0057),(Item=(Type=Bone,Name=\"spine_01\"),Weight=0.0371),(Item=(Type=Bone,Name=\"spine_05\"),Weight=0.0729),(Item=(Type=Bone,Name=\"spine_02\"),Weight=0.1712),(Item=(Type=Bone,Name=\"spine_03\"),Weight=0.3526),(Item=(Type= ...(truncated)"
    Weight=@RigVMFunction_MathBoolToFloat.Result
   )
   > &RigUnit_ModifyTransforms(
    ItemToModify=((Item=(Type=Bone,Name="base_pendant_for_aim"),Transform=(Rotation=(X=-0.0523,Y=0,Z=0,W=0.9986),Translation=(X=0,Y=0,Z=-0.85),Scale3D=(X=1,Y=1,Z=1))))
    WeightMinimum=0
    Weight=@RigVMFunction_MathBoolToFloat.Result
   )
   > &RigUnit_AimConstraintLocalSpaceOffset(
    Child=(Type=Bone,Name="base_pendant")
    Filter=(bX=true,bY=True,bZ=True)
    Parents=((Item=(Type=Bone,Name="base_pendant_for_aim"),Weight=1))
    WorldUp=(Target=(X=0,Y=0,Z=1),Kind=Direction,Space=(Type=Bone,Name="base_pendant_for_aim"))
    Weight=@RigVMFunction_MathBoolToFloat.Result
   )
  ]}
]
variables [
 CalfRot_L {type=double, default=0}
 CalfRot_R {type=double, default=0}
 IsFrontEnd {type=bool, default=False}
]
