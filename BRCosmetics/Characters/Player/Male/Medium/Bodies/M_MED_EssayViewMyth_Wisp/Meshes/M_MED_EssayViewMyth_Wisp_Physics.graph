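# Physics asset for the lanyard on the M_MED_EssayViewMyth_Wisp male medium player body cosmetic.
# Two dynamic-named lanyard capsules, left and right, hang from a tiny base capsule with locked twist and swing limit angles.
# The base and both thighs link to the pelvis with free angular motion.
# Larger pelvis and thigh capsules approximate the lower body, likely as collision for the swinging lanyard.
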PhysicsAsset /BRCosmetics/Characters/Player/Male/Medium/Bodies/M_MED_EssayViewMyth_Wisp/Meshes/M_MED_EssayViewMyth_Wisp_Physics
bodies {
 base_lanyard [{shapeType=Capsule, length=0.505, radius=0.505, shapeName=base_lanyard_capsule, rotation=<rot 0 -90 0>}]
 dyn_lanyard_l [{shapeType=Capsule, length=10, radius=1, shapeName=dyn_lanyard_r_capsule, center=<10 0 1>, rotation=<rot 90 0 0>}]
 dyn_lanyard_r [{shapeType=Capsule, length=10, radius=1, shapeName=dyn_lanyard_r_capsule, center=<10 0 1>, rotation=<rot 90 0 0>}]
 pelvis [{
   shapeType=Capsule
   length=7.8461
   radius=11.5895
   shapeName=pelvis_capsule
   center=<-3.4322 -0.9808 -0.142>
   rotation=<rot -6.2784 -90 -0.1834>
  }]
 thigh_l [{shapeType=Capsule, length=4, radius=7, shapeName=thigh_r_capsule, center=<-3.3186 4.0178 6.2113>, rotation=<rot -0.1947 -94.3984 90>}]
 thigh_r [{shapeType=Capsule, length=4, radius=7, shapeName=thigh_r_capsule, center=<3.2541 -4.8559 -4.7673>, rotation=<rot -0.1947 -94.3984 90>}]
}
constraints [
 {
  bone1Name=thigh_l
  bone2Name=pelvis
  swing1LimitDegrees=45
  swing1Motion=Free
  swing2LimitDegrees=45
  swing2Motion=Free
  twistLimitDegrees=45
  twistMotion=Free
 }
 {
  bone1Name=base_lanyard
  bone2Name=pelvis
  swing1LimitDegrees=45
  swing1Motion=Free
  swing2LimitDegrees=45
  swing2Motion=Free
  twistLimitDegrees=45
  twistMotion=Free
 }
 {bone1Name=dyn_lanyard_l, bone2Name=base_lanyard, swing1LimitDegrees=25, swing2LimitDegrees=25, twistLimitDegrees=45, twistMotion=Locked}
 {
  bone1Name=thigh_r
  bone2Name=pelvis
  swing1LimitDegrees=45
  swing1Motion=Free
  swing2LimitDegrees=45
  swing2Motion=Free
  twistLimitDegrees=45
  twistMotion=Free
 }
 {bone1Name=dyn_lanyard_r, bone2Name=base_lanyard, swing1LimitDegrees=45, swing2LimitDegrees=25, twistLimitDegrees=0, twistMotion=Locked}
]
