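# Physics asset for the top mesh of the M_MED_EssayViewMyth_Wisp player cosmetic, covering a pendant and the upper spine.
# Capsule bodies form a three-link dyn_pendant chain hanging from base_pendant, which attaches to spine_05, plus larger torso capsules on spine_03 to spine_05.
# The pendant links have tight swing and twist limits, with twist locked at the first link, while the base and spine joints rotate freely.
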PhysicsAsset /BRCosmetics/Characters/Player/Male/Medium/Bodies/M_MED_EssayViewMyth_Wisp/Meshes/M_MED_EssayViewMyth_Wisp_Top_Physics
bodies {
 base_pendant [{shapeType=Capsule, length=0.505, radius=0.505, shapeName=base_pendant_capsule, rotation=<rot 0 -90 0>}]
 dyn_pendant_1 [{shapeType=Capsule, length=10, radius=5.1, shapeName=dyn_pendant_1_capsule, center=<5 0 4.2>, rotation=<rot 0 0 90>}]
 dyn_pendant_2 [{shapeType=Capsule, length=9, radius=2, shapeName=dyn_pendant_1_capsule, center=<2 0 1.5>, rotation=<rot 0 0 90>}]
 dyn_pendant_3 [{shapeType=Capsule, length=4, radius=4, shapeName=dyn_pendant_3_capsule, center=<5 0 2.5>, rotation=<rot 0 0 90>}]
 spine_03 [{shapeType=Capsule, length=15.8461, radius=12.1901, shapeName=spine_03_capsule, rotation=<rot -0.0009 -90 0.0004>}]
 spine_04 [{shapeType=Capsule, length=15.8461, radius=12.1901, shapeName=spine_03_capsule, rotation=<rot -0.0009 -90 0.0004>}]
 spine_05 [{shapeType=Capsule, length=15.8461, radius=10, shapeName=spine_03_capsule, rotation=<rot -0.0009 -90 0.0004>}]
}
constraints [
 {bone1Name=dyn_pendant_3, bone2Name=dyn_pendant_2, swing1LimitDegrees=25, swing2LimitDegrees=25, twistLimitDegrees=2.5}
 {bone1Name=dyn_pendant_2, bone2Name=dyn_pendant_1, swing1LimitDegrees=25, swing2LimitDegrees=25, twistLimitDegrees=2.5}
 {bone1Name=dyn_pendant_1, bone2Name=base_pendant, swing1LimitDegrees=10, swing2LimitDegrees=25, twistLimitDegrees=5, twistMotion=Locked}
 {
  bone1Name=base_pendant
  bone2Name=spine_05
  swing1LimitDegrees=45
  swing1Motion=Free
  swing2LimitDegrees=45
  swing2Motion=Free
  twistLimitDegrees=45
  twistMotion=Free
 }
 {
  bone1Name=spine_04
  bone2Name=spine_03
  swing1LimitDegrees=45
  swing1Motion=Free
  swing2LimitDegrees=45
  swing2Motion=Free
  twistLimitDegrees=45
  twistMotion=Free
 }
 {
  bone1Name=spine_05
  bone2Name=spine_04
  swing1LimitDegrees=45
  swing1Motion=Free
  swing2LimitDegrees=45
  swing2Motion=Free
  twistLimitDegrees=45
  twistMotion=Free
 }
]
